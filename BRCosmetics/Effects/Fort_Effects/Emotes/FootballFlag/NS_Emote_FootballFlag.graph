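# Niagara system for the football flag emote: throws a single mesh flag with a puff of sprite smoke defined alongside it.
# Flag launches up and sideways under gravity, with age-gated collision, a local orientation-fix module and a shrink near death.
# Smoke particles drift, drag, grow, step through sub-images and fade.
# Flag generates a death event, but Smoke's burst is disabled and no receiving event handler appears here.
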NiagaraSystem /BRCosmetics/Effects/Fort_Effects/Emotes/FootballFlag/NS_Emote_FootballFlag
dependencies {usedDataInterfaces [&NiagaraDataInterfaceCurve], usedDynamicInputs [], usedModules [&SystemState]}
emitters {
 Flag {
  emitterUpdateScript [
   &EmitterState()
   &SpawnBurst_Instantaneous(Spawn Count=1, Use Spawn Probability=false)
  ]
  particleSpawnScript [
   /Niagara/Modules/Spawn/Initialization/InitializeParticle(
    Lifetime=2
    Module.Write Color=true
    Module.Write Lifetime=true
    Module.Write Mass=true
    Module.Write Material Random=true
    Module.Write Position=false
    Module.Write Scale=true
    Module.Write SpriteRotation=false
    Module.Write SpriteSize=true
    Position=&SimulationPosition()
    Sprite Size=<5 5>
   )
   /Niagara/Modules/Spawn/Orientation/V2/InitialMeshOrientation(Orientation Axis:Vector3f=<-1 0 0>, Orientation Vector=<0 0 1>, Use Orientation Vector=true, Use Rotation Vector=false) {enabled=false}
   /Niagara/Modules/Spawn/Orientation/V3/InitialMeshOrientation(Orientation Axis=<-1 0 0>, Orientation Vector=<0 0 1>, Use Orientation Vector=true, Use Rotation Vector=false) {module=InitialMeshOrientation002}
   &AddVelocity(Velocity=<0 -100 500>)
  ]
  particleUpdateScript [
   &ParticleState()
   &GravityForce()
   &Collision(
    Age Colliding Particles=false
    Collision Enabled=&Boolean_AndOperation(Bool1=true, Bool2=&SetBoolByFloatComparison(A=0.25, B=$Particles.Age, Comparison Type=::NewEnumerator7))
    Friction=1
    Module.Control Rest State Via Surface Normal=false
    Module.EnableMaxCollisionCount=false
    Module.Randomize Collision Normal Vector=false
    Radius Calculation Type=::NewEnumerator1
    Restitution=0.25
   )
   &SolveForcesAndVelocity(Clamp Velocity=false, Limit Acceleration=false)
   &ScaleMeshSize(Scale Factor=&VectorFromFloat(Value=&FloatFromCurve(FloatCurve=DataInterface() {propertyValues {Curve {keys [<at 0.9 1 RCIM_Cubic>,<at 1 0 RCIM_Cubic>]}}})))
   /SetVariables(Particles.FlagPosition:NiagaraPosition=$Particles.Position) {isSetParameters=true}
   &GenerateDeathEvent()
   &self:FlagOrientFix() {module=FlagOrientFix}
   &CameraOffset(Camera Offset Amount=3)
  ]
  rendererClasses [&NiagaraMeshRendererProperties]
  renderers [{rendererClass=&NiagaraMeshRendererProperties, rendererIndex=0}]
 }
 Smoke {
  emitterUpdateScript [
   &EmitterState()
   &SpawnBurst_Instantaneous(Spawn Count=30, Spawn Time=2, Use Spawn Probability=false) {enabled=false, module=SpawnBurst_Instantaneous001}
  ]
  particleSpawnScript [
   &InitializeParticle(
    Color=&UniformRangedLinearColor(Minimum=<rgba 0.4323 0.4323 0.4323 0.35>, Maximum=<rgba 0.5255 0.5255 0.5255 0.5>)
    Lifetime=&UniformRangedFloat(Minimum=0.25, Maximum=0.5)
    Module.Write Color=true
    Module.Write Lifetime=true
    Module.Write Mass=true
    Module.Write Material Random=true
    Module.Write Position=false
    Module.Write Scale=false
    Module.Write SpriteRotation=true
    Module.Write SpriteSize=true
    Position=&SimulationPosition()
    Sprite Rotation=&UniformRangedFloat(Minimum=0, Maximum=360)
    Sprite Size=&Vector2DFromFloat(Value=&UniformRangedFloat(Minimum=5, Maximum=10))
   )
   &AccelerationForce(Acceleration=&UniformRangedVector(Minimum=<0 0 0>, Maximum=<0 0 50>))
   &AddVelocity(Velocity=&UniformRangedVector(Minimum=<-100 -100 -100>, Maximum=<100 100 100>))
   &ApplyInitialForces()
  ]
  particleUpdateScript [
   &ParticleState()
   &Drag(Drag=3, Use Linear Drag=true, Use Rotational Drag=false)
   &SolveForcesAndVelocity(Clamp Velocity=false, Limit Acceleration=false)
   &ScaleSpriteSize(Scale Factor=&Vector2DFromFloat(Value=&FloatFromCurve(FloatCurve=DataInterface() {propertyValues {Curve {keys [
          {arriveTangent=1.491, interpMode=RCIM_Cubic, leaveTangent=1.491, tangentMode=RCTM_User, time=-0.0046, value=1.1969}
          <at 1.0011 3.4809 RCIM_Cubic>
         ]}}})))
   /SetVariables(Particles.SubImageIndex=&FloatFromCurve(FloatCurve=DataInterface() {propertyValues {Curve {keys [
         {arriveTangent=29.8623, interpMode=RCIM_Cubic, leaveTangent=29.8623, tangentMode=RCTM_User, time=-0.0046, value=5.0556}
         <at 1.0011 22.1689 RCIM_Cubic>
        ]}}})) {isSetParameters=true}
   &DynamicMaterialParameters(
    Index 0 Param 1=&FloatFromCurve(FloatCurve=DataInterface() {propertyValues {Curve {keys [<at -0.0046 1.0113 RCIM_Cubic>,<at 1.0011 0.002>]}}})
    Index 0 Param 4=&FloatFromCurve(FloatCurve=DataInterface() {propertyValues {Curve {keys [<at 0 0 RCIM_Cubic>,<at 1 0.5>]}}})
    Module.Param0WriteEnabled=true
    Module.Param1WriteEnabled=false
    Module.Param2WriteEnabled=false
    Module.Param3WriteEnabled=true
   ) {module=DynamicMatParams}
   &CameraOffset(Camera Offset Amount=50)
   &ScaleColor(
    Scale Alpha=&FloatFromCurve(FloatCurve=DataInterface() {propertyValues {Curve {keys [
         {arriveTangent=-3.1757, interpMode=RCIM_Cubic, leaveTangent=-3.1757, tangentMode=RCTM_User, time=0.5, value=1.0113}
         {arriveTangent=-1.5121, interpMode=RCIM_Cubic, leaveTangent=-1.5121, tangentMode=RCTM_User, time=1.0011, value=0.002}
        ]}}})
    ScaleA=true
    ScaleRGB=false
   )
  ]
  rendererClasses [&NiagaraSpriteRendererProperties]
  renderers [{rendererClass=&NiagaraSpriteRendererProperties}]
 }
}
summary {
 systemName=NS_Emote_FootballFlag
 emitters [
  {emitterName=Flag, rendererClasses [&NiagaraMeshRendererProperties]}
  {emitterName=Smoke, rendererClasses [&NiagaraSpriteRendererProperties]}
 ]
 userVariables []
}
